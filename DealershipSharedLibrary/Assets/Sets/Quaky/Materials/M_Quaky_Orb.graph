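# Unlit translucent material for a Quaky orb effect on mesh and Niagara mesh particles, with an animated watery, cyan-tinted glow.
# A panning flow normal map distorts a scrolling water texture, which an inverted Fresnel masks.
# Particle color and a dark-teal-to-cyan blend driven by a dynamic parameter feed an emissive-adjustment function.
# Parameterized sine waves, masked along V, build an animated vector with no output connection shown.
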Material /DealershipSharedLibrary/Assets/Sets/Quaky/Materials/M_Quaky_Orb
graph {
 properties {
  bAutomaticallySetUsageInEditor=false
  bUsedWithMeshParticles=true
  bUsedWithNiagaraMeshParticles=true
  blendMode=BLEND_Translucent
  shadingModel=MSM_Unlit
  usedShadingModels=Unlit
 }
 nodes [
  n1 = ParticleColor()
  n2 = TextureCoordinate() {uTiling=3, vTiling=3}
  n3 = Multiply(A=Time())
  n4 = ComponentMask(None=TextureSample(UVs=Panner(Coordinate=@n2, Time=@n3) {speedX=0.3, speedY=0.3}) {samplerType=SAMPLERTYPE_Normal, texture=/DealershipSharedLibrary/Assets/Sets/Quaky/Textures/T_Quaky_Flow_03}.RGB) {g=true, r=true}
  n5 = Panner(Coordinate=Add(A=TextureCoordinate(), B=Multiply(A=@n4))) {speedX=-1, speedY=-0.3}
  n6 = Multiply(A=Fresnel() {exponent=0.6})
  n7 = Multiply(A=Constant() {r=2}, B=Power(Base=OneMinus(None=@n6)))
  n8 = Multiply(
   A=Multiply(A=TextureSample(UVs=@n5) {texture=/DealershipSharedLibrary/Assets/Sets/Quaky/Textures/T_Quaky_Water_01}.RGB, B=Clamp(None=@n7))
   B=Constant() {r=2}
  )
  n9 = LinearInterpolate(
   A=Constant3Vector() {constant=<rgba 0 0.06 0.06 0>}
   B=Constant3Vector() {constant=<rgba 0.3 1 1 0>}
   Alpha=DynamicParameter() {defaultValue=<rgba 0 1 1 1>, paramNames [Fresnel Color, Param2, Param3, Param4]}.Fresnel Color
  ) {constAlpha=0}
  MaterialFunctionCall(Emissive=Add(A=Multiply(A=@n1.RGB, B=@n8), B=Add(A=@n8, B=@n9))) {materialFunction=&MF_FX_EmissiveAdjustments}
  n10 = Constant()
  n11 = Divide(A=ComponentMask(None=TextureCoordinate()) {r=true}, B=ScalarParameter() {defaultValue=0.8, parameterName=CountWaves2})
  n12 = Multiply(A=ScalarParameter() {defaultValue=0.5, parameterName=Horizontal Speed}, B=OneMinus(None=Time()))
  n13 = Multiply(A=ScalarParameter() {defaultValue=0.5, parameterName=Horizontal Speed 2}, B=OneMinus(None=Time()))
  n14 = Multiply(A=Sine(None=Add(A=@n11, B=@n12)) {period=0.5}, B=Sine(None=Add(A=@n11, B=@n13)) {period=0.5})
  n15 = Divide(A=ComponentMask(None=TextureCoordinate()) {g=true}, B=ScalarParameter() {defaultValue=0.8, parameterName=CountWaves2})
  n16 = Time()
  n17 = Add(A=@n15, B=Multiply(A=ScalarParameter() {defaultValue=1, parameterName=Vertical Speed}, B=OneMinus(None=@n16)))
  n18 = Time()
  n19 = Add(A=@n15, B=Multiply(A=ScalarParameter() {defaultValue=-0.5, parameterName=Vertical Speed 2}, B=OneMinus(None=@n18)))
  n20 = Multiply(A=OneMinus(None=Sine(None=@n17)), B=OneMinus(None=Sine(None=@n19)))
  n21 = Multiply(
   A=AppendVector(A=Power(Base=@n14) {constExponent=1.2}, B=Power(Base=@n20) {constExponent=1.2})
   B=ScalarParameter() {defaultValue=2, parameterName=Strength}
  )
  n22 = ComponentMask(None=TextureCoordinate()) {g=true}
  n23 = OneMinus(None=Power(Base=@n22, Exp=ScalarParameter() {defaultValue=1, parameterName=Power Middle 2}))
  n24 = Multiply(A=Multiply(A=Multiply(A=@n23, B=@n22), B=ScalarParameter() {defaultValue=6, parameterName=Mult Power 2 Vertical}))
  Multiply(A=AppendVector(A=AppendVector(A=@n10, B=@n10), B=Multiply(A=@n21, B=@n24)), B=Constant() {r=10})
 ]
}
