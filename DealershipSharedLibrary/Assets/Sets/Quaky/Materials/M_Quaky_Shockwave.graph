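# Translucent, unlit, two-sided material for a Quaky shockwave effect on Niagara mesh particles.
# An inverted-V UV gradient shapes a soft body and sharp bright edge, broken up by panning sunlight and caustic textures and scaled by particle alpha.
# A second caustic layer, warped by two panning water normal maps with tunable velocity and intensity, enriches opacity before depth fade.
# Particle color plus the edge highlight feeds an emissive adjustment function.
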Material /DealershipSharedLibrary/Assets/Sets/Quaky/Materials/M_Quaky_Shockwave
graph {
 properties {
  bAutomaticallySetUsageInEditor=false
  bCastRayTracedShadows=false
  bUsedWithNiagaraMeshParticles=true
  blendMode=BLEND_Translucent
  shadingModel=MSM_Unlit
  twoSided=true
  usedShadingModels=Unlit
 }
 nodes [
  n1 = ParticleColor()
  n2 = OneMinus(None=ComponentMask(None=TextureCoordinate()) {g=true})
  n3 = Power(Base=Multiply(A=@n2) {constB=0.45}) {constExponent=1.75}
  n4 = TextureCoordinate() {uTiling=4}
  n5 = Power(Base=TextureSample(UVs=Panner(Coordinate=@n4) {speedY=-0.3}) {texture=/DealershipSharedLibrary/Assets/Sets/Quaky/Textures/T_Quaky_Sunlight}.B) {constExponent=0.3}
  n6 = Power(Base=@n2) {constExponent=20}
  n7 = Multiply(A=@n6) {constB=1.5}
  n8 = Multiply(A=@n7, B=@n6)
  n9 = TextureSample(UVs=Panner(Coordinate=TextureCoordinate()) {speedY=-0.3}) {texture=/DealershipSharedLibrary/Assets/Sets/Quaky/Textures/T_Quaky_Caustic}
  n10 = Multiply(A=@n6, B=Multiply(B=Power(Base=@n9.B) {constExponent=1}) {constA=20})
  n11 = Multiply(A=@n1.A, B=Add(A=Multiply(A=@n3, B=@n5), B=Add(A=@n8, B=@n10)))
  n12 = Constant() {r=20}
  n13 = Panner(
   Coordinate=Multiply(A=TextureCoordinate() {vTiling=0.25}, B=@n12)
   Time=Multiply(A=Time(), B=ScalarParameter() {defaultValue=1, parameterName=Distortion Velocity})
  )
  n14 = Multiply(A=@n12, B=TextureCoordinate())
  n15 = TextureSample(UVs=Panner(Coordinate=@n14) {speedX=0.05}) {samplerType=SAMPLERTYPE_Normal, texture=/DealershipSharedLibrary/FX/Textures/Normal/T_VCVFX_Water03Swizzled_Noise}
  n16 = TextureSample(UVs=Panner(Coordinate=@n14) {speedX=-0.05, speedY=0.05}) {samplerType=SAMPLERTYPE_Normal, texture=&T_VCVFX_Water03Swizzled_Noise}
  n17 = Multiply(
   A=ComponentMask(None=Add(A=@n15.RGB, B=@n16.RGB)) {g=true, r=true}
   B=ScalarParameter() {defaultValue=0.15, parameterName=Distortion Intensity}
  )
  n18 = Multiply(A=@n11, B=TextureSample(UVs=Add(A=@n13, B=@n17)) {texture=&T_Quaky_Caustic}.RGB)
  MaterialFunctionCall(Opacity In=Clamp(None=Add(A=@n11, B=@n18)), DepthFade Distance=Constant() {r=4000}) {materialFunction=&MF_VCVFX_DepthFade}
  MaterialFunctionCall(Emissive=Add(A=@n1.RGB, B=@n7)) {materialFunction=&MF_FX_EmissiveAdjustments}
 ]
}
